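# Masked skeletal-mesh material for the Vorta vehicle wheel, adding an animated pulse to MF_WheelParent_Simple.
# A panning, squared sine mask driven by FXSpeed combines with the tire mask's red channel to modulate base color and emissive from black.
# The parent's VertexColorR output limits where this effect applies.
# The modified attributes then pass through the shared vehicle damage and de-rez transition functions.
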Material /VehicleCosmetics/Wheels/Vorta/Materials/M_Wheel_Vorta
graph {
 properties {
  bAutomaticallySetUsageInEditor=false
  bUseMaterialAttributes=true
  bUsedWithMeshDeformer=true
  bUsedWithSkeletalMesh=true
  blendMode=BLEND_Masked
  pixelDepthOffsetMode=PDOM_Legacy
  usedShadingModels=Default Lit
 }
 nodes [
  n1 = MaterialFunctionCall() {materialFunction=&MF_WheelParent_Simple}
  n2 = GetMaterialAttributes(None=@n1.MaterialAttributes) {attributeGetTypes [#1, #2]}
  n3 = Constant3Vector()
  n4 = ComponentMask(None=@n1.TireRGBAMask) {r=true}
  n5 = Multiply(A=@n4, B=NamedRerouteUsage())
  n6 = LinearInterpolate(A=@n2.Emissive Color, B=LinearInterpolate(A=@n3, B=Multiply(A=@n2.Emissive Color, B=@n2.Base Color), Alpha=@n5), Alpha=@n1.VertexColorR)
  n7 = LinearInterpolate(A=@n2.Base Color, B=LinearInterpolate(A=@n3, B=Multiply(A=@n4, B=@n2.Base Color), Alpha=@n5), Alpha=@n1.VertexColorR)
  MaterialFunctionCall(In=MaterialFunctionCall(Material=SetMaterialAttributes!(MaterialAttributes=@n2.MaterialAttributes, Emissive Color=@n6, Base Color=@n7) {attributeSetTypes [#1, #2]}) {materialFunction=&MF_BaseVehicle_Damage}.Result) {materialFunction=&MF_VehicleDeRez_Transition}
  n8 = TextureCoordinate() {uTiling=0.5}
  n9 = Multiply(A=Time(), B=ScalarParameter() {defaultValue=1, group=5_Special Effects, parameterName=FXSpeed})
  n10 = Sine(None=ComponentMask(None=Panner(Coordinate=@n8, Time=@n9) {speedX=-0.1}) {r=true})
  n11 = Multiply(A=@n10, B=@n10)
  NamedRerouteDeclaration(None=Add(A=Clamp(None=@n11)) {constB=0.1}) {name=PannerFXMask, nodeColor=<rgba 0 1 0.2734 1>}
 ]
}
